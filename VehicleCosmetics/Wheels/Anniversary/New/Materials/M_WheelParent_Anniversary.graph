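# Parent material for Anniversary vehicle wheels, adding metal-part overrides and a glowing rim to MF_WheelParent_Simple.
# Vertex color red and tire-mask alpha select metal parts; Roughness_Metal_Parts feeds Metallic and Metallic_Metal_Parts feeds Roughness, apparently swapped.
# Switchable RimLight emissive combines a rim mask, CastomColor and two rotating noise samples, limited to low-red vertices.
# Results pass through vehicle damage, then de-rez transition functions.
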Material /VehicleCosmetics/Wheels/Anniversary/New/Materials/M_WheelParent_Anniversary
graph {
 properties {
  bAutomaticallySetUsageInEditor=false
  bUseMaterialAttributes=true
  bUsedWithMeshDeformer=true
  bUsedWithSkeletalMesh=true
  pixelDepthOffsetMode=PDOM_Legacy
  usedShadingModels=Default Lit
 }
 nodes [
  Constant()
  n1 = MaterialFunctionCall() {materialFunction=&MF_WheelParent_Simple}
  n2 = GetMaterialAttributes(None=@n1.MaterialAttributes) {attributeGetTypes [
    #1
    #2
    #3
    #4
   ]}
  n3 = VectorParameter() {parameterName=Color_Metal_Parts, defaultValue=<rgba 1 1 1 1>}
  n4 = OneMinus(None=ComponentMask(None=@n1.TireRGBAMask) {a=true})
  n5 = ScalarParameter() {defaultValue=0.5, parameterName=Roughness_Metal_Parts}
  n6 = ScalarParameter() {defaultValue=1, parameterName=Metallic_Metal_Parts}
  n7 = SetMaterialAttributes!(
   MaterialAttributes=@n2.MaterialAttributes
   Emissive Color=LinearInterpolate(A=@n2.Emissive Color, B=NamedRerouteUsage(), Alpha=OneMinus(None=@n1.VertexColorR))
   Base Color=LinearInterpolate(A=LinearInterpolate(A=@n2.Base Color, B=@n3.RGB, Alpha=@n1.VertexColorR), B=@n2.Base Color, Alpha=@n4)
   Metallic=LinearInterpolate(A=LinearInterpolate(A=@n2.Metallic, B=@n5, Alpha=@n1.VertexColorR), B=@n2.Metallic, Alpha=@n4)
   Roughness=LinearInterpolate(A=LinearInterpolate(A=@n2.Roughness, B=@n6, Alpha=@n1.VertexColorR), B=@n2.Roughness, Alpha=@n4)
  ) {attributeSetTypes [
    #1
    #2
    #3
    #4
   ]}
  MaterialFunctionCall(In=MaterialFunctionCall(Material=@n7) {materialFunction=&MF_BaseVehicle_Damage}.Result) {materialFunction=&MF_VehicleDeRez_Transition}
  n8 = TextureSampleParameter2D() {
   parameterName=RimEmmisive
   samplerType=SAMPLERTYPE_LinearColor
   texture=/VehicleCosmetics/Wheels/Anniversary/New/Textures/T_Wheel_Anniversary_Rim_M
  }
  n9 = Clamp(None=ScalarParameter() {defaultValue=1, parameterName=RimEmissiveGlow})
  n10 = Constant() {r=7}
  n11 = VectorParameter() {parameterName=CastomColor, defaultValue=<rgba 0 0.0821 0.76 0>}
  n12 = Add(A=Constant() {r=-0.1}, B=OneMinus(None=Multiply(A=@n10, B=@n11.RGB)))
  n13 = Multiply(A=Constant() {r=0.5}, B=@n11.RGB)
  n14 = TextureCoordinate() {uTiling=0.2, vTiling=0.2}
  n15 = TextureCoordinate() {uTiling=0.3, vTiling=0.3}
  n16 = Add(
   A=TextureSample(UVs=Rotator(Coordinate=@n14)) {texture=/VehicleCosmetics/Wheels/_SharedTextures/FX/ContrastNoiseGradient_T}.B
   B=TextureSample(UVs=Rotator(Coordinate=@n15)) {texture=&ContrastNoiseGradient_T}.R
  )
  n17 = LinearInterpolate(A=LinearInterpolate(A=Clamp(None=@n12), B=@n13, Alpha=@n11.A), B=@n13, Alpha=Clamp(None=ConstantBiasScale(None=@n16) {bias=-0.15, scale=3}))
  n18 = Constant()
  n19 = StaticSwitchParameter(True=Multiply(A=Multiply(A=@n8.R, B=@n9), B=Clamp(None=@n17)), False=@n18) {defaultValue=true, parameterName=RimLight}
  n20 = VertexColor()
  NamedRerouteDeclaration(None=Multiply(A=LinearInterpolate(A=@n19, B=@n18, Alpha=@n20.R), B=Constant() {r=4})) {name=Emissive, nodeColor=<rgba 1 0.75 0 1>}
  OneMinus(None=@n20.R)
 ]
}
